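# Translucent, lit alpha-composite material for arrows in the VR Editor UI, with per-pixel surface lighting and screen-space reflections.
# A constant cyan-green color is scaled by the Alpha parameter, dimmed, and passed through EyeAdaptationInverse to counter auto-exposure.
# Alpha also scales a Fresnel-driven blend on the pixel normal, giving a slight view-angle variation, so one parameter fades both terms.
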Material /Engine/VREditor/UI/ArrowMaterial
graph {
 properties {
  bScreenSpaceReflections=true
  bUseTranslucencyVertexFog=false
  blendMode=BLEND_AlphaComposite
  pixelDepthOffsetMode=PDOM_Legacy
  translucencyLightingMode=TLM_SurfacePerPixelLighting
  translucencyPass=MTP_BeforeDOF
  usedShadingModels=Default Lit
 }
 nodes [
  Constant()
  Constant() {r=0.1}
  n1 = Constant3Vector() {constant=<rgba 0 1 0.8796 0>}
  n2 = ScalarParameter() {defaultValue=1, parameterName=Alpha}
  EyeAdaptationInverse(LightValueInput=Divide(A=Multiply(A=@n1, B=@n2)) {constB=4}, AlphaInput=Constant() {r=1})
  Multiply(A=Constant() {r=1}, B=@n2)
  n3 = PixelNormalWS()
  Multiply(A=@n2, B=LinearInterpolate(Alpha=Fresnel(Normal=@n3)) {constA=0.9})
 ]
}
