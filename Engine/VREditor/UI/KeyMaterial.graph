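# KeyMaterial is an unlit translucent material in the VR Editor UI that builds a glowing, rim-weighted color, green by default.
# The BaseColor and GlowAmount parameters are multiplied together and by a Fresnel term driven by the world-space pixel normal.
# The result passes through EyeAdaptationInverse to counteract auto exposure.
# The same Fresnel also drives a separate scalar lerp, leaving a nonzero floor at facing angles.
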Material /Engine/VREditor/UI/KeyMaterial
graph {
 properties {blendMode=BLEND_Translucent, pixelDepthOffsetMode=PDOM_Legacy, shadingModel=MSM_Unlit, usedShadingModels=Unlit}
 nodes [
  n1 = VectorParameter() {parameterName=BaseColor, defaultValue=<rgba 0.1271 0.98 0.299 1>}
  n2 = ScalarParameter() {defaultValue=10, parameterName=GlowAmount}
  n3 = Fresnel(Normal=PixelNormalWS()) {baseReflectFraction=0.2, exponent=2}
  EyeAdaptationInverse(LightValueInput=Multiply(A=Multiply(A=@n1.RGB, B=@n2), B=@n3), AlphaInput=Constant() {r=1})
  LinearInterpolate(Alpha=@n3) {constA=0.25}
 ]
}
