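# Default Lit material for the VR Editor content window, previewed on SM_ContentWindow_01, that resizes the frame through vertex-color-driven offsets.
# Each vertex color channel masks one scalar: AddTop (R), AddRight (G), AddBottom (B), AddLeft (A).
# These build a local Y/Z vector, transformed out of local space, presumably for world position offset.
# A near-black color and UE logo-masked 0.4–0.475 blend suggest base color and roughness.
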Material /Engine/VREditor/UI/M_ContentWindow
graph {
 properties {pixelDepthOffsetMode=PDOM_Legacy, previewMesh=&/SM_ContentWindow_01, usedShadingModels=Default Lit}
 nodes [
  Constant3Vector() {constant=<rgba 0.0072 0.0076 0.01 1>}
  LinearInterpolate(Alpha=TextureSample() {texture=/Engine/VREditor/Devices/Vive/UE_Logo}.R) {constA=0.4, constB=0.475}
  n1 = Constant()
  n2 = VertexColor()
  n3 = ScalarParameter() {parameterName=AddRight}
  n4 = Add(A=Multiply(A=@n2.G, B=Multiply(A=@n3) {constB=-1}), B=Multiply(A=@n2.A, B=ScalarParameter() {parameterName=AddLeft}))
  n5 = Multiply(A=@n2.R, B=ScalarParameter() {parameterName=AddTop})
  n6 = Multiply(A=@n2.B, B=Multiply(A=ScalarParameter() {parameterName=AddBottom}) {constB=-1})
  Transform(None=AppendVector(A=AppendVector(A=@n1, B=@n4), B=Add(A=@n5, B=@n6))) {transformSourceType=TRANSFORMSOURCE_Local}
 ]
}
