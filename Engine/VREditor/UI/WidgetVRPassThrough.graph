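# Unlit, opaque, one-sided material for VR Editor UI widgets that displays the SlateUI texture parameter.
# TwoSidedSign blends BackColor for back faces with the texture for front faces, then TintColorAndOpacity multiplies it.
# EyeAdaptationInverse counteracts auto exposure on that color.
# Separate branches compute alpha from OpacityFromTexture and a per-channel sRGB-to-linear version of the color; the graph shows no material output links.
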Material /Engine/VREditor/UI/WidgetVRPassThrough
graph {
 properties {
  editorOnlyData="<self>:Widget3DCameraPassThrough_Opaque_OneSidedEditorOnlyData"
  pixelDepthOffsetMode=PDOM_Legacy
  shadingModel=MSM_Unlit
  usedShadingModels=Unlit
 }
 nodes [
  n1 = VectorParameter() {parameterName=BackColor}
  n2 = TextureSampleParameter2D() {parameterName=SlateUI, samplerSource=SSM_Clamp_WorldGroupSettings, texture=/Engine/EngineResources/WhiteSquareTexture}
  n3 = Add(A=Multiply(A=TwoSidedSign()) {constB=0.5}) {constB=0.5}
  n4 = VectorParameter() {parameterName=TintColorAndOpacity, defaultValue=<rgba 1 1 1 1>}
  EyeAdaptationInverse(LightValueInput=Multiply(A=LinearInterpolate(A=@n1.RGB, B=@n2.RGB, Alpha=@n3), B=@n4.RGB), AlphaInput=Constant() {r=1})
  Multiply(A=LinearInterpolate(B=@n2.A, Alpha=ScalarParameter() {defaultValue=1, parameterName=OpacityFromTexture, sliderMax=1}) {constA=1}, B=@n4.A)
  n5 = MaterialFunctionCall(Float3=TextureSampleParameter2D() {parameterName=SlateUI, samplerSource=SSM_Clamp_WorldGroupSettings, texture=&WhiteSquareTexture}.RGB) {materialFunction=&BreakOutFloat3Components}
  n6 = MaterialFunctionCall!(sRGB=@n5.R) {materialFunction=/Engine/Functions/Engine_MaterialFunctions01/ImageAdjustment/sRGBToLinear}
  n7 = MaterialFunctionCall(sRGB=@n5.G) {materialFunction=&sRGBToLinear}
  n8 = MaterialFunctionCall(sRGB=@n5.B) {materialFunction=&sRGBToLinear}
  n9 = Multiply(A=TwoSidedSign()) {constB=0.5}
  Multiply(
   A=LinearInterpolate(
    A=VectorParameter() {parameterName=BackColor}.RGB
    B=MaterialFunctionCall(X=@n6.Linear, Y=@n7.Linear, Z=@n8.Linear) {materialFunction=&MakeFloat3}.Result
    Alpha=Add(A=@n9) {constB=0.5}
   )
   B=Constant() {r=1}
  )
  VectorParameter() {parameterName=TintColorAndOpacity, defaultValue=<rgba 1 1 1 1>}
 ]
}
